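# Creative property editor icon-picker swatch that shows a character item or icon option inside an outlined border.
# Set Represented Icon async-loads the character definition's preview, falling back to a simple-tagged texture or material-parameter texture with default.
# SetRaised animates or snaps scale and yellow border; the snap's scale selection is inverted.
# PreConstruct sizes from forced or curve-table values, using the PC row on non-Windows platforms; SetSelected is empty.
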WidgetBlueprint /FortniteCoreUI/InGame/Creative/PropertyEditor/Icon_Picker/CharacterItem/Icon_Picker/CreativeCharacterItemPickerSwatch
animations [Anim_SelectSwatch {
  binding_count=2
  length_seconds=1
  object=&self:Anim_SelectSwatch
  sequence {bindings [
    FortLazyImage_Swatch {
     id=#1
     possessed_class=FortLazyImage
     tracks [{
       type=MovieScene2DTransformTrack
       sections [{
         type=MovieScene2DTransformSection
         end_frame=20
         start_frame=0
         channels [
          Scale.X {type=MovieSceneScriptingFloatChannel, keys [[0,0.75],[20,0.875]]}
          Scale.Y {type=MovieSceneScriptingFloatChannel, keys [[0,0.75],[20,0.875]]}
         ]
        }]
      }]
    }
    Image_Border {
     id=#2
     possessed_class=CommonBorder
     tracks [{
       type=MovieSceneColorTrack
       display_name=Brush Color
       sections [{
         type=MovieSceneColorSection
         end_frame=20
         start_frame=0
         channels [
          Color.R {type=MovieSceneScriptingFloatChannel, const [1,1]}
          Color.G {type=MovieSceneScriptingFloatChannel, const [0.9216,1]}
          Color.B {type=MovieSceneScriptingFloatChannel, const [0,1]}
          Color.A {type=MovieSceneScriptingFloatChannel, keys [[0,0],[20,1]]}
         ]
        }]
      }]
    }
   ]}
 }]
graphs [
 EventGraph {nodes [
   n1 = Set Represented Icon!() -> (OutputDelegate, then, Icon Option:"Playlist Option Icon Value Structure (by ref)") {is=title K2Node_CustomEvent}
   n2 = Break Playlist Option Icon Value!(PlaylistOptionIconValue:"Playlist Option Icon Value Structure (by ref)"=@n1.Icon Option) -> (Value:Creative Icon Option Structure, DisplayName:Text, OptionValueName:String, GrantedContextTags:Gameplay Tag Container Structure) {is=title K2Node_BreakStruct}
   > n3 = Async Load Asset(Asset=To Soft Object Reference(SoftObjectPath=MakeSoftObjectPath!(PathString:String=@n2.OptionValueName) -> (ReturnValue:Soft Object Path Structure)))
   n4 = Cast To AthenaCharacterItemDefinition!(execute=@n3.Completed, Object=@n3) -> (then, CastFailed, AsAthena Character Item Definition:Athena Character Item Definition Object Reference) {is=title K2Node_DynamicCast}
   > SetBrushFromSoftTexture!(
    execute
    self:Image Object Reference=Get FortLazyImage_Swatch() -> (FortLazyImage_Swatch:Fort Lazy Image Object Reference)
    SoftTexture:Texture 2D Soft Object Reference=GetSmallPreviewImage!(ItemDefinition:Item Definition Base Object Reference=@n4) -> (ReturnValue:Texture 2D Soft Object Reference)
    bMatchSize:bool=false
   ) -> (then)
   n5 = Branch(
    execute=@n4.CastFailed
    Condition="Equal (GameplayTagContainer)"(
     A:"Gameplay Tag Container Structure (by ref)"=Get SimpleIconTag() -> (SimpleIconTag:Gameplay Tag Container Structure)
     B:"Gameplay Tag Container Structure (by ref)"=@n2.GrantedContextTags
    ) -> (ReturnValue:bool=~) {is=title K2Node_PromotableOperator}
   )
   n6 = Break Creative Icon Option!(CreativeIconOption:"Creative Icon Option Structure (by ref)"=@n2) -> (SmallIcon:Texture 2D Soft Object Reference, LargeIcon:Texture 2D Soft Object Reference) {is=title K2Node_BreakStruct}
   > SetBrushFromSoftTexture(self=Get FortLazyImage_Swatch(), SoftTexture=@n6, bMatchSize=false)
   n7 = Branch(execute=@n5.else, Condition=IsValidSoftObjectReference(SoftObjectReference=@n6))
   > n8 = Async Load Asset(Asset=@n6)
   n9 = Cast To Texture2D(execute=@n8.Completed, Object=@n8)
   > SetBrushFromMaterial(self=Get FortLazyImage_Swatch(), Material=Get DefaultIconMaterial() -> (DefaultIconMaterial:Material Interface Object Reference))
   > n10 = GetDynamicMaterial(self=Get FortLazyImage_Swatch())
   > SetTextureParameterValue(
    self=@n10
    ParameterName=Texture
    Value=Select(
     Option 0:Texture 2D Object Reference=Get DefaultSwatchTexture() -> (DefaultSwatchTexture:Texture 2D Object Reference)
     Option 1:Texture 2D Object Reference=@n9
     Index=IsValid(Object=@n9)
    ) -> (ReturnValue:Texture 2D Object Reference)
   )
   SetBrushFromMaterial(execute=@n7.else, self=Get FortLazyImage_Swatch(), Material=Get DefaultIconMaterial())
   > n11 = GetDynamicMaterial(self=Get FortLazyImage_Swatch())
   > SetTextureParameterValue(self=@n11, ParameterName=Texture, Value=Get DefaultSwatchTexture())
   Event PreConstruct()
   n12 = GetValueAtLevel!(Input:"Scalable Float Structure (by ref)"=Get ScalableEntrySizePC() -> (ScalableEntrySizePC:Scalable Float Structure), Level:float=0) -> (ReturnValue:float)
   n13 = MakeVector2D(X=@n12, Y=@n12)
   n14 = GetValueAtLevel(Input=Get ScalableEntrySizeOther() -> (ScalableEntrySizeOther:Scalable Float Structure), Level=0)
   n15 = MakeVector2D(X=@n14, Y=@n14)
   n16 = "Not Equal Exactly (String)"(A=GetPlatformName!() -> (ReturnValue:String), B=Windows)
   > SetDesiredSizeOverride!(
    execute
    self:Image Object Reference=Get FortLazyImage_Swatch()
    DesiredSize:Vector 2D Structure=Select(
     Option 0:Vector 2D Structure=Select(Option 0=@n15, Option 1=@n13, Index=@n16)
     Option 1:Vector 2D Structure=Get ForceDesiredSize() -> (ForceDesiredSize:Vector 2D Structure)
     Index=Get bUseForceDesiredSize() -> (bUseForceDesiredSize:bool)
    ) -> (ReturnValue:Vector 2D Structure)
   ) -> (then)
   Event Construct()
   > SetRenderTransform!(
    execute
    self:Widget Object Reference=Get FortLazyImage_Swatch()
    InTransform:Widget Transform Structure=Make Widget Transform!(
     Translation_X:double=0
     Translation_Y:double=0
     Scale:Vector 2D Structure=Get Swatch Scale Unselected() -> (Swatch Scale Unselected:Vector 2D Structure)
     Shear:Vector 2D Structure=(X=0,Y=0)
     Angle:float=0
    ) -> (WidgetTransform:Widget Transform Structure) {is=title K2Node_MakeStruct}
   ) -> (then)
  ]}
 Set Icon to Represent {nodes [
   n1 = Set Icon to Represent!() -> (then, IconOption:Playlist Option Icon Value Structure) {is=title K2Node_FunctionEntry}
   n2 = Break Playlist Option Icon Value(PlaylistOptionIconValue=@n1)
   > n3 = Branch(Condition="Equal (GameplayTagContainer)"(A=Get SimpleIconTag(), B=@n2.GrantedContextTags) -> (ReturnValue=~))
   n4 = Break Creative Icon Option(CreativeIconOption=@n2)
   > SetBrushFromSoftTexture(self=Get FortLazyImage_Swatch(), SoftTexture=@n4, bMatchSize=false)
   SetBrushFromMaterial(execute=@n3.else, self=Get FortLazyImage_Swatch(), Material=Get DefaultIconMaterial())
   > n5 = GetDynamicMaterial(self=Get FortLazyImage_Swatch())
   n6 = Resolve Soft Reference!(Input:Texture 2D Soft Object Reference=@n4) -> (Output:Texture 2D Object Reference)
   > SetTextureParameterValue(
    self=@n5
    ParameterName=Texture
    Value=Select(
     Option 0:Texture 2D Object Reference=Get DefaultSwatchTextureIcon() -> (DefaultSwatchTextureIcon:Texture 2D Object Reference)
     Option 1:Texture 2D Object Reference=@n6
     Index=IsValid(Object=@n6)
    ) -> (ReturnValue:Texture 2D Object Reference)
   )
  ]}
 SetRaised {nodes [
   n1 = SetRaised!() -> (then, bIsRaised:bool=~, bPlayAnimation:bool=~) {is=title K2Node_FunctionEntry}
   > Branch(Condition="Not Equal (Boolean)"!(A:bool=@n1, B:bool=Get bIsSwatchRaised() -> (bIsSwatchRaised:bool)) -> (ReturnValue:bool) {is=title K2Node_CallFunction})
   > Set bIsSwatchRaised(bIsSwatchRaised:bool=Get bIsRaised() -> (bIsRaised:bool)) -> (Output_Get:bool)
   > n2 = Branch(Condition=@n1.bPlayAnimation)
   > PlayAnimation(
    InAnimation=Get Anim_SelectSwatch() -> (Anim_SelectSwatch:Widget Animation Object Reference)
    PlayMode=Select(Option 0:EUMGSequencePlayMode Enum=Reverse, Option 1:EUMGSequencePlayMode Enum, Index=Get bIsSwatchRaised()) -> (ReturnValue:EUMGSequencePlayMode Enum)
    PlaybackSpeed=4
   )
   n3 = Get Swatch Scale Unselected()
   n4 = Get Swatch Scale Selected() -> (Swatch Scale Selected:Vector 2D Structure)
   n5 = Get bIsSwatchRaised()
   SetRenderTransform(
    execute=@n2.else
    self=Get FortLazyImage_Swatch()
    InTransform=Make Widget Transform(
     Translation_X=0
     Translation_Y=0
     Scale=Select(Option 0:Vector 2D Structure=@n4, Option 1:Vector 2D Structure=@n3, Index=@n5) -> (ReturnValue:Vector 2D Structure)
     Shear=(X=0,Y=0)
     Angle=0
    )
   )
   n6 = Get bIsSwatchRaised()
   > SetBrushColor(
    self=Get Image_Border() -> (Image_Border:Common Border Object Reference)
    InBrushColor=MakeColor!(R:float=1, G:float=0.9216, B:float=0, A:float=SelectFloat(A=1, B=0, bPickA=@n6)) -> (ReturnValue:Linear Color Structure)
   )
  ]}
 SetSelected {nodes [SetSelected!() -> (then, bIsSelected:bool=~) {is=title K2Node_FunctionEntry}]}
]
variables [
 bIsSwatchRaised {type=boolean}
 bUseForceDesiredSize {type=boolean}
 DefaultIconMaterial {type=MaterialInterface, default=/Game/Athena/UI/InGame/Creative/PropertyEditor/Icon_Picker/Materials/M_UI_IconPicker_Icon_Inst}
 DefaultSwatchTexture {type=Texture2D, default=/Game/Athena/UI/InGame/Creative/PropertyEditor/Icon_Picker/Textures/T_UI_IconPicker_NONE}
 DefaultSwatchTextureIcon {type=Texture2D, default=/Game/Global/Textures/Default/Blanks/T_BlackUserInterface}
 ForceDesiredSize {type=Vector2D, default=<80 80>}
 ScalableEntrySizeOther {
  type=ScalableFloat
  default {
   registryType=None
   value=1
   curve {curveTable=/Game/Athena/Playlists/Creative/AthenaCompositeGD_Creative, rowName=Creative.UI.CharacterItemPicker.DrawSize.Other}
  }
 }
 ScalableEntrySizePC {
  type=ScalableFloat
  default {
   registryType=None
   value=1
   curve {curveTable=&AthenaCompositeGD_Creative, rowName=Creative.UI.CharacterItemPicker.DrawSize.PC}
  }
 }
 SimpleIconTag {type=GameplayTagContainer, default {gameplayTags [{tagName=Creative.UserOptions.Icon.UseSimpleIcon}]}}
 Swatch Scale Selected {type=Vector2D, default=<0.875 0.875>}
 Swatch Scale Unselected {type=Vector2D, default=<0.75 0.75>}
]
widgets {
 info {inheritedWidgetCount=0, parentClass=&UserWidget, rootWidgetClass=&Overlay, widgetCount=3}
 widgets [
  Overlay_98 {widgetClassPath=&Overlay, widget=Overlay_98, properties {visibility=HitTestInvisible}}
  Image_Border {
   widgetClassPath=&CommonBorder
   bIsVariable=true
   parent=Overlay_98
   widget=Image_Border
   properties {
    horizontalAlignment=HAlign_Center
    style=/FortUILibrary/Styles/Border/Border-1pxOutline.Border-1pxOutline_C
    verticalAlignment=VAlign_Bottom
    brushColor=<rgba 1 0.9216 0 0>
    padding=<ltrb 0 0 0 0>
   }
   slot {class=OverlaySlot}
  }
  FortLazyImage_Swatch {
   widgetClassPath=&FortLazyImage
   bIsVariable=true
   parent=Image_Border
   widget=FortLazyImage_Swatch
   properties {brush {
     bIsDynamicallyLoaded=false
     drawAs=Image
     imageType=NoImage
     mirroring=NoMirror
     resourceName=None
     resourceObject=&M_UI_IconPicker_Icon_Inst
     tiling=NoTile
     imageSize=<64 64>
     margin=<ltrb 0 0 0 0>
     outlineSettings {
      bUseBrushTransparency=false
      roundingType=HalfHeightRadius
      width=0
      color {colorUseRule=UseColor_Specified, specifiedColor=<rgba 0 0 0 0>}
      cornerRadii=<0 0 0 0>
     }
     tintColor {colorUseRule=UseColor_Specified, specifiedColor=<rgba 1 1 1 1>}
     uVRegion {bIsValid=false, max=<0 0>, min=<0 0>}
    }}
   slot {class=BorderSlot, properties {horizontalAlignment=HAlign_Center, verticalAlignment=VAlign_Bottom, padding=<ltrb 0 0 0 0>}}
  }
 ]
}
