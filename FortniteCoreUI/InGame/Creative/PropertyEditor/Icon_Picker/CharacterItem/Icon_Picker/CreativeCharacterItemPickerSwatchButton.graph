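# Creative icon picker list-entry button that wraps a character item swatch, built on CommonButtonBase.
# Receiving an IconOptionObjectWrapper stores its icon option, shows it on the swatch and syncs raised and selected visuals without animation.
# Hover and selection changes animate the raise, selection also takes focus, and a click fires OnSwatchCommitted only if already selected.
# A helper matches icon objects by case-insensitive display name and granted gameplay tags.
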WidgetBlueprint /FortniteCoreUI/InGame/Creative/PropertyEditor/Icon_Picker/CharacterItem/Icon_Picker/CreativeCharacterItemPickerSwatchButton
graphs [
 EventGraph {nodes [
   n1 = Event OnListItemObjectSet()
   > n2 = Cast To IconOptionObjectWrapper!(execute, Object=@n1.ListItemObject) -> (then, CastFailed, AsIcon Option Object Wrapper:Icon Option Object Wrapper Object Reference) {is=title K2Node_DynamicCast}
   > n3 = Set RepresentedIconOptionSwatch(RepresentedIconOptionSwatch:Playlist Option Icon Value Structure=Get IconOption!(self:Icon Option Object Wrapper Object Reference=@n2) -> (IconOption:Playlist Option Icon Value Structure)) -> (Output_Get:Playlist Option Icon Value Structure)
   n4 = Get CreativeIconPickerSwatch() -> (CreativeIconPickerSwatch:Creative Character Item Picker Swatch Object Reference)
   > Set Represented Icon!(execute, self:Creative Character Item Picker Swatch Object Reference=@n4, Icon Option:"Playlist Option Icon Value Structure (by ref)"=@n3) -> (then) {is=title K2Node_CallFunction}
   > n5 = Set bIsSelected(bIsSelected:bool=IsListItemSelected!(UserListEntry:User List Entry Interface) -> (ReturnValue:bool)) -> (Output_Get:bool)
   > SetRaised!(execute, self:Creative Character Item Picker Swatch Object Reference=@n4, bIsRaised:bool=@n5, bPlayAnimation:bool=false) -> (then)
   > SetSelected!(execute, self:Creative Character Item Picker Swatch Object Reference=@n4, bIsSelected:bool=@n5) -> (then)
   Event On Hovered()
   > SetRaised(self=Get CreativeIconPickerSwatch(), bIsRaised=true, bPlayAnimation=true)
   Event On Unhovered()
   > Branch(Condition=NOT Boolean(A=Get bIsSelected() -> (bIsSelected:bool)))
   > SetRaised(self=Get CreativeIconPickerSwatch(), bIsRaised=false, bPlayAnimation=true)
   Event On Clicked()
   > Branch(Condition=Get bIsSelected())
   > Call On Swatch Committed!(execute:Exec, self:Creative Character Item Picker Swatch Button Object Reference) -> (then:Exec) {is=title K2Node_CallDelegate}
   Event On Selected()
   > SetFocus()
   > n6 = Set bIsSelected(bIsSelected=true)
   n7 = Set bIsSelected(execute=@n10.then, bIsSelected=false)
   n8 = Merge Execution with Data!(Option A:Exec=@n6.then, Option A Data:bool=@n6, Option B:Exec=@n7.then, Option B Data:bool=@n7) -> (Then:Exec, Data:bool=~) {is=title K2Node_MacroInstance}
   n9 = Get CreativeIconPickerSwatch()
   > SetRaised(self=@n9, bIsRaised=@n8, bPlayAnimation=true)
   > SetSelected(self=@n9, bIsSelected=@n8)
   n10 = Event On Deselected()
  ]}
 Is Represented Icon Object {nodes [
   n1 = Is Represented Icon Object!() -> (then, InIconObject:Icon Option Object Wrapper Object Reference) {is=title K2Node_FunctionEntry}
   > n2 = Is Valid(InputObject=@n1)
   Return Node!(execute=@n2.Is Not Valid, IsObject:bool=false)
   n3 = Break Playlist Option Icon Value!(PlaylistOptionIconValue:"Playlist Option Icon Value Structure (by ref)"=Get IconOption(self=@n1)) -> (DisplayName:Text, GrantedContextTags:Gameplay Tag Container Structure) {is=title K2Node_BreakStruct}
   n4 = "To String (Text)"(InText=@n3)
   n5 = Break Playlist Option Icon Value(PlaylistOptionIconValue=Get RepresentedIconOptionSwatch() -> (RepresentedIconOptionSwatch:Playlist Option Icon Value Structure))
   n6 = "To String (Text)"(InText=@n5)
   > Return Node(IsObject=AND Boolean(
     A="Equal, Case Insensitive (String)"(A=@n4, B=@n6)
     B="Equal (GameplayTagContainer)"(A:"Gameplay Tag Container Structure (by ref)"=@n3.GrantedContextTags, B:"Gameplay Tag Container Structure (by ref)"=@n5.GrantedContextTags) -> (ReturnValue:bool=~) {is=title K2Node_PromotableOperator}
    ))
  ]}
 OnSwatchCommitted {nodes [OnSwatchCommitted() {is=title K2Node_FunctionEntry}]}
]
variables [
 bIsSelected {type=boolean}
 OnSwatchCommitted {}
 RepresentedIconOptionSwatch {
  type=PlaylistOptionIconValue
  default {
   displayName=""
   optionValueName=""
   grantedContextTags {gameplayTags []}
   optionalDisplayIcon {largeIcon=null, smallIcon=null}
   value {largeIcon=null, smallIcon=null}
  }
 }
]
widgets {
 info {inheritedWidgetCount=1, parentClass=&CommonButtonBase, rootWidgetClass=&Overlay, widgetCount=3}
 widgets [
  Overlay_20 {widgetClassPath=&Overlay, widget=Overlay_20}
  CreativeIconPickerSwatch {
   widgetClassPath=&/CreativeCharacterItemPickerSwatch
   bIsVariable=true
   parent=Overlay_20
   widget=CreativeIconPickerSwatch
   slot {class=OverlaySlot}
  }
  InputActionWidget {
   widgetClassPath=&CommonActionWidget
   bInherited=true
   bIsVariable=true
   parent=Overlay_20
   widget=InputActionWidget
   properties {
    iconRimBrush {
     bIsDynamicallyLoaded=false
     drawAs=Image
     imageType=NoImage
     mirroring=NoMirror
     resourceName=None
     resourceObject=/FortUILibrary/Foundation/Buttons/Materials/M-Button-Console-YellowOutline
     tiling=NoTile
     imageSize=<32 32>
     margin=<ltrb 0 0 0 0>
     outlineSettings {
      bUseBrushTransparency=false
      roundingType=HalfHeightRadius
      width=0
      color {colorUseRule=UseColor_Specified, specifiedColor=<rgba 0 0 0 0>}
      cornerRadii=<0 0 0 0>
     }
     tintColor {colorUseRule=UseColor_Specified, specifiedColor=<rgba 1 1 1 1>}
     uVRegion {bIsValid=false, max=<0 0>, min=<0 0>}
    }
    renderTransform {angle=0, scale=<1 1>, shear=<0 0>, translation=<0 8>}
   }
   slot {class=OverlaySlot, properties {horizontalAlignment=HAlign_Center, verticalAlignment=VAlign_Bottom}}
  }
 ]
}
