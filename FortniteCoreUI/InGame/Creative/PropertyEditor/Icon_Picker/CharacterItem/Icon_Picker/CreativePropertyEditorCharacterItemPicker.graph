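# Character item icon picker dialog for the Creative property editor, built on FortCreativePropertyEditorColorPickerDialog.
# Selecting a tile view swatch updates the header name, which alternates with a search box, and binds that swatch's commit event, rebinding as entries recycle.
# Accept or checkmark broadcasts the selection, preferring its unfiltered-list index; Back broadcasts closure.
# Tile size follows platform via scalable floats, while input method changes adjust search visibility and dialog height.
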WidgetBlueprint /FortniteCoreUI/InGame/Creative/PropertyEditor/Icon_Picker/CharacterItem/Icon_Picker/CreativePropertyEditorCharacterItemPicker
animations [Anim_IntroOutro {
  binding_count=1
  length_seconds=0.25
  object=&self:Anim_IntroOutro
  sequence {bindings [Overlay_18 {
     id=#1
     possessed_class=Overlay
     tracks [
      {
       type=MovieScene2DTransformTrack
       sections [{
         type=MovieScene2DTransformSection
         end_frame=5
         start_frame=0
         channels [
          Translation.X {type=MovieSceneScriptingFloatChannel, keys [[0,-440],[5,0]]}
          Translation.Y {type=MovieSceneScriptingFloatChannel, const [0,1]}
          Angle {type=MovieSceneScriptingFloatChannel, const [0,1]}
          Scale.X {type=MovieSceneScriptingFloatChannel, const [1,1]}
          Scale.Y {type=MovieSceneScriptingFloatChannel, const [1,1]}
          Shear.X {type=MovieSceneScriptingFloatChannel, const [0,1]}
          Shear.Y {type=MovieSceneScriptingFloatChannel, const [0,1]}
         ]
        }]
      }
      {
       type=MovieSceneFloatTrack
       display_name=Render Opacity
       sections [{type=MovieSceneFloatSection, end_frame=5, start_frame=0, channels [None {type=MovieSceneScriptingFloatChannel, keys [[0,0],[3,0],[5,1]]}]}]
      }
     ]
    }]}
 }]
graphs [
 BP_GetDesiredFocusTarget {nodes [
   Get Desired Focus Target()
   > Return Node!(
    execute
    ReturnValue:Widget Object Reference=Get CurrentIconSwatchButton() -> (CurrentIconSwatchButton:Creative Character Item Picker Swatch Button Object Reference)
   )
  ]}
 Commit Selected Swatch {nodes [
   Commit Selected Swatch() {is=title K2Node_FunctionEntry}
   > n1 = Cast To IconOptionObjectWrapper!(
    execute
    Object=Get Selected Item!(self:List View Object Reference=Get TileView_Swatches() -> (TileView_Swatches:Common Tile View Object Reference)) -> (ReturnValue:Object Reference) {is=title K2Node_CallFunction}
   ) -> (then, CastFailed, AsIcon Option Object Wrapper:Icon Option Object Wrapper Object Reference) {is=title K2Node_DynamicCast}
   n2 = GetItemIndexFromUnfilteredItemList!(Item:Object Reference=@n1) -> (ReturnValue:int=0)
   n3 = Get TileView_Swatches()
   n4 = Break Playlist Option Icon Value!(PlaylistOptionIconValue:"Playlist Option Icon Value Structure (by ref)"=Get IconOption!(self:Icon Option Object Wrapper Object Reference=@n1) -> (IconOption:Playlist Option Icon Value Structure)) -> (Value:Creative Icon Option Structure, DisplayName:Text, OptionValueName:String, GrantedContextTags:Gameplay Tag Container Structure) {is=title K2Node_BreakStruct}
   > Call On Swatch Selection Committed!(
    execute:Exec
    self:Creative Property Editor Character Item Picker Object Reference
    Index:int=SelectInt(
     A=@n2
     B=GetIndexForItem!(self:List View Object Reference=@n3, Item:Object Reference=@n1) -> (ReturnValue:int=0)
     bPickA="Not Equal (Integer)"(A=@n2, B=-1)
    )
    DisplayText:Text=TextToUpper!(InText:"Text (by ref)"=@n4.DisplayName) -> (ReturnValue:Text)
    IconValue:Creative Icon Option Structure=@n4
   ) -> (then:Exec) {is=title K2Node_CallDelegate}
  ]}
 EventGraph {nodes [
   n1 = "On Button Clicked (UserActionMenuInputButton_Accept)"!() -> (OutputDelegate, then, Button:Common Button Legacy Object Reference) {is=title K2Node_ComponentBoundEvent}
   Commit Selected Swatch(execute=[@n1.then, @n2.then]) {is=title K2Node_CallFunction}
   "On Button Clicked (UserActionMenuInputButton_Back)"!() -> (OutputDelegate, then, Button:Common Button Legacy Object Reference) {is=title K2Node_ComponentBoundEvent}
   > Call On Icon Picker Closed!(execute:Exec, self:Creative Property Editor Character Item Picker Object Reference) -> (then:Exec) {is=title K2Node_CallDelegate}
   "On Clicked (Button_Search)"() {is=title K2Node_ComponentBoundEvent}
   > HandleSearchInputVisibility!(execute, self, ShowSearchText:bool=true) -> (then)
   > SetFocus(self=Get EditableText_Search() -> (EditableText_Search:Text Box Object Reference))
   n2 = "On Clicked (Button_CheckmarkBackground)"() {is=title K2Node_ComponentBoundEvent}
   n3 = "On Hovered (Button_Search)"() {is=title K2Node_ComponentBoundEvent}
   HandleSearchInputVisibility(execute=[@n3.then, @n15.then], ShowSearchText=true)
   n4 = "On Item Selection Changed (TileView_Swatches)"!() -> (OutputDelegate, then, Item:Object Reference, bIsSelected:bool) {is=title K2Node_ComponentBoundEvent}
   > Branch(Condition=@n4.bIsSelected)
   > Unbind Current Swatch Button()
   > n5 = Cast To IconOptionObjectWrapper(Object=@n4.Item)
   > n6 = Set MostRecentlySelectedIconItem(MostRecentlySelectedIconItem:Icon Option Object Wrapper Object Reference=@n5) -> (Output_Get:Icon Option Object Wrapper Object Reference)
   > Update Header!(execute, self, SelectedIconOption:Icon Option Object Wrapper Object Reference=@n6) -> (then)
   > Update Current Swatch Button()
   > HandleSearchInputVisibility(ShowSearchText=false)
   n7 = "On Entry Generated (TileView_Swatches)"!() -> (OutputDelegate, then, Widget:User Widget Object Reference) {is=title K2Node_ComponentBoundEvent}
   > n8 = Cast To CreativeCharacterItemPickerSwatchButton!(execute, Object=@n7.Widget) -> (then, CastFailed, AsCreative Character Item Picker Swatch Button:Creative Character Item Picker Swatch Button Object Reference) {is=title K2Node_DynamicCast}
   > Branch(Condition=Is Represented Icon Object!(
     self:Creative Character Item Picker Swatch Button Object Reference=@n8
     InIconObject:Icon Option Object Wrapper Object Reference=Get MostRecentlySelectedIconItem() -> (MostRecentlySelectedIconItem:Icon Option Object Wrapper Object Reference)
    ) -> (IsObject:bool))
   > n9 = Set CurrentIconSwatchButton(CurrentIconSwatchButton:Creative Character Item Picker Swatch Button Object Reference=@n8) -> (Output_Get:Creative Character Item Picker Swatch Button Object Reference)
   > Bind Event to On Swatch Committed(self:Creative Character Item Picker Swatch Button Object Reference=@n9, Delegate=Create Event()) {is=title K2Node_AddDelegate}
   n10 = "On Entry Released (TileView_Swatches)"!() -> (OutputDelegate, then, Widget:User Widget Object Reference) {is=title K2Node_ComponentBoundEvent}
   > n11 = Cast To CreativeCharacterItemPickerSwatchButton(Object=@n10.Widget)
   > Branch(Condition=Is Represented Icon Object(self=@n11, InIconObject=Get MostRecentlySelectedIconItem()))
   > Unbind Current Swatch Button()
   Event PreConstruct()
   > SetupCheckmarkBackgroundStyle()
   > n12 = Branch(Condition="Not Equal Exactly (String)"(A=GetPlatformName!() -> (ReturnValue:String), B=Windows))
   n13 = GetValueAtLevel!(Input:"Scalable Float Structure (by ref)"=Get ScalableEntrySizeOther() -> (ScalableEntrySizeOther:Scalable Float Structure), Level:float=0) -> (ReturnValue:float)
   > SetEntryHeight!(execute, self:Tile View Object Reference=Get TileView_Swatches(), NewHeight:float=@n13) -> (then)
   > SetEntryWidth!(execute, self:Tile View Object Reference=Get TileView_Swatches(), NewWidth:float=@n13) -> (then)
   n14 = GetValueAtLevel(Input=Get ScalableEntrySizePC() -> (ScalableEntrySizePC:Scalable Float Structure), Level=0)
   SetEntryHeight(execute=@n12.else, self=Get TileView_Swatches(), NewHeight=@n14)
   > SetEntryWidth(self=Get TileView_Swatches(), NewWidth=@n14)
   n15 = Event OnSearchInput() {is=title K2Node_Event}
   Event OnInitialized()
   > SetInputAction!(
    execute
    self:Common Action Widget Object Reference=Get SearchActionWidget() -> (SearchActionWidget:Common Action Widget Object Reference)
    InputActionRow:Data Table Row Handle Structure=Get SearchInputRowHandle() -> (SearchInputRowHandle:Data Table Row Handle Structure)
   ) -> (then)
   Event OnOutroAnimFinished() {is=title K2Node_Event}
   > Update Checkmark Background for Controller Select!(execute, self, bIsSelected:bool=false) -> (then)
   Event OnOpen() {is=title K2Node_Event}
   > Update Checkmark Background for Controller Select(bIsSelected=false)
   Event OnClose() {is=title K2Node_Event}
   > Call On Icon Picker Closed()
   n16 = Event OnInputMethodChanged!() -> (OutputDelegate, then, InputType:ECommonInputType Enum) {is=title K2Node_Event}
   n17 = "Equal (Enum)"(A=@n16.InputType, B=MouseAndKeyboard)
   > SetVisibility(
    self=Get SizeBox_SearchInput() -> (SizeBox_SearchInput:Size Box Object Reference)
    InVisibility=Select(Option 0=HitTestInvisible, Option 1=Collapsed, Index=@n17)
   )
   > SetHeightOverride(self=Get SizeBox_WidgetSize() -> (SizeBox_WidgetSize:Size Box Object Reference), InHeightOverride=SelectFloat(A=350, B=418, bPickA=@n17))
  ]}
 HandleInputMethodChanges {nodes [
   n1 = HandleInputMethodChanges!() -> (then, bNewInputType:ECommonInputType Enum) {is=title K2Node_FunctionEntry}
   n2 = "Equal (Enum)"(A=@n1, B=MouseAndKeyboard)
   > SetVisibility(self=Get SizeBox_SearchInput(), InVisibility=Select(Option 0=HitTestInvisible, Option 1=Collapsed, Index=@n2))
   > SetHeightOverride(self=Get SizeBox_WidgetSize(), InHeightOverride=SelectFloat(A=350, B=418, bPickA=@n2))
  ]}
 HandleSearchInputVisibility {nodes [
   n1 = HandleSearchInputVisibility!() -> (then, ShowSearchText:bool=~) {is=title K2Node_FunctionEntry}
   > SetVisibility(self=Get EditableText_Search(), InVisibility=Select(Option 0=Collapsed, Option 1=SelfHitTestInvisible, Index=@n1))
   > SetVisibility(
    self=Get Text_SwatchName() -> (Text_SwatchName:Common Text Object Reference)
    InVisibility=Select(Option 0=SelfHitTestInvisible, Option 1=Collapsed, Index=@n1)
   )
  ]}
 OnIconPickerClosed {nodes [OnIconPickerClosed() {is=title K2Node_FunctionEntry}]}
 OnSwatchSelectionCommitted {nodes [OnSwatchSelectionCommitted!() -> (then, Index:int, DisplayText:Text, IconValue:Creative Icon Option Structure) {is=title K2Node_FunctionEntry}]}
 SelectSwatch {nodes [
   n1 = SelectSwatch!() -> (then, Swatch Index:int) {is=title K2Node_FunctionEntry}
   > Branch(Condition="integer >= integer"(A=@n1))
   n2 = Get TileView_Swatches()
   > ScrollIndexIntoView!(execute, self:List View Object Reference=@n2, Index:int=@n1) -> (then)
   > SetSelectedIndex!(execute, self:List View Object Reference=@n2, Index:int=@n1) -> (then)
  ]}
 SetupCheckmarkBackgroundStyle {nodes [
   SetupCheckmarkBackgroundStyle() {is=title K2Node_FunctionEntry}
   > SetStyle!(
    execute
    self:Button Object Reference=Get Button_CheckmarkBackground() -> (Button_CheckmarkBackground:Button Object Reference)
    InStyle:"Button Style Structure (by ref)"=Make Button Style!(
     Normal:Slate Brush Structure=Get CheckBackground_Normal() -> (CheckBackground_Normal:Slate Brush Structure)
     Hovered:Slate Brush Structure=Get CheckBackground_Hovered() -> (CheckBackground_Hovered:Slate Brush Structure)
     Pressed:Slate Brush Structure=Get CheckBackground_Pressed() -> (CheckBackground_Pressed:Slate Brush Structure)
     Disabled:Slate Brush Structure=(TintColor=(SpecifiedColor=(R=1,G=1,B=1,A=1),ColorUseRule=UseColor_Specified),DrawAs=NoDrawType,Tiling=NoTile,Mirroring=NoMirror,ImageType=NoImage,ImageSize=(X=0,Y=0),Margin=(Left=0,Top=0,Right=0,Bottom=0),ResourceObject=None,OutlineSettings=(CornerRadii=(X=0,Y=0,Z=0,W=0),Color=(SpecifiedColor=(R=0,G=0,B=0,A=0),ColorUseRule=UseColor_Specified),Width=0,RoundingType=HalfHeightRadius,bUseBrushTransparency=False),UVRegion=(Min=(X=0,Y=0),Max=(X=0,Y=0),bIsValid=False),bIsDynamicallyLoaded=False,ResourceName="")
     NormalForeground:Slate Color Structure=(SpecifiedColor=(R=1,G=0,B=1,A=1),ColorUseRule=UseColor_Foreground)
     HoveredForeground:Slate Color Structure=(SpecifiedColor=(R=1,G=0,B=1,A=1),ColorUseRule=UseColor_Foreground)
     PressedForeground:Slate Color Structure=(SpecifiedColor=(R=1,G=0,B=1,A=1),ColorUseRule=UseColor_Foreground)
     DisabledForeground:Slate Color Structure=(SpecifiedColor=(R=1,G=0,B=1,A=1),ColorUseRule=UseColor_Foreground)
     NormalPadding:Margin Structure=(Left=0,Top=0,Right=0,Bottom=0)
     PressedPadding:Margin Structure=(Left=0,Top=0,Right=0,Bottom=0)
     PressedSlateSound:Slate Sound Structure=(ResourceObject=None)
     ClickedSlateSound:Slate Sound Structure=(ResourceObject=None)
     HoveredSlateSound:Slate Sound Structure=(ResourceObject=None)
    ) -> (ButtonStyle:Button Style Structure) {is=title K2Node_MakeStruct}
   ) -> (then)
  ]}
 Unbind Current Swatch Button {nodes [
   Unbind Current Swatch Button() {is=title K2Node_FunctionEntry}
   n1 = Get CurrentIconSwatchButton()
   > Is Valid(InputObject=@n1)
   > Unbind all Events from On Swatch Committed!(execute:Exec, self:Creative Character Item Picker Swatch Button Object Reference=@n1) -> (then:Exec) {is=title K2Node_ClearDelegate}
   > Set CurrentIconSwatchButton()
  ]}
 Update Checkmark Background for Controller Select {nodes [
   n1 = Update Checkmark Background for Controller Select!() -> (then, bIsSelected:bool=~) {is=title K2Node_FunctionEntry}
   > Set members in Button Style!(
    execute:Exec
    StructRef:"Button Style Structure (by ref)"=Get WidgetStyle!(self:Button Object Reference=Get Button_CheckmarkBackground()) -> (WidgetStyle:Button Style Structure)
    Normal:Slate Brush Structure=Select(Option 0:Slate Brush Structure=Get CheckBackground_Normal(), Option 1:Slate Brush Structure=Get CheckBackground_Pressed(), Index=@n1) -> (ReturnValue:Slate Brush Structure)
   ) -> (then:Exec, StructOut:"Button Style Structure (by ref)") {is=title K2Node_SetFieldsInStruct}
  ]}
 Update Current Swatch Button {nodes [
   Update Current Swatch Button() {is=title K2Node_FunctionEntry}
   > Set CurrentIconSwatchButton()
   > n1 = For Each Loop with Break!(
    Exec:Exec
    Array:Array of User Widget Object References=GetDisplayedEntryWidgets!(self:List View Base Object Reference=Get TileView_Swatches()) -> (ReturnValue:Array of User Widget Object References)
    Break:Exec=@n4.then
   ) -> (LoopBody:Exec, Array Element:User Widget Object Reference, Array Index:int, Completed:Exec)
   > n2 = Cast To CreativeCharacterItemPickerSwatchButton(Object=@n1)
   > Branch(Condition=Is Represented Icon Object(self=@n2, InIconObject=Get MostRecentlySelectedIconItem()))
   > n3 = Set CurrentIconSwatchButton(CurrentIconSwatchButton=@n2)
   > n4 = Bind Event to On Swatch Committed(self=@n3, Delegate=Create Event())
  ]}
 Update Header {nodes [
   n1 = Update Header!() -> (then, SelectedIconOption:Icon Option Object Wrapper Object Reference) {is=title K2Node_FunctionEntry}
   > "SetText (Text)"(
    self=Get Text_SwatchName()
    InText=TextToUpper(InText=Get!(self:Icon Option Object Wrapper Object Reference=@n1) -> (
      IconOption_Value:Creative Icon Option Structure
      IconOption_DisplayName:Text
      IconOption_OptionValueName:String
      IconOption_GrantedContextTags:Gameplay Tag Container Structure
     ).IconOption_DisplayName)
   )
  ]}
]
variables [
 CheckBackground_Hovered {
  type=SlateBrush
  default {
   bIsDynamicallyLoaded=false
   drawAs=Image
   imageType=NoImage
   mirroring=NoMirror
   resourceName=None
   resourceObject=/Game/Athena/UI/InGame/Creative/PropertyEditor/Color_Picker/T_Icon_ColorPicker_action_bg
   tiling=NoTile
   imageSize=<94 81>
   margin=<ltrb 0 0 0 0>
   outlineSettings {
    bUseBrushTransparency=false
    roundingType=HalfHeightRadius
    width=0
    color {colorUseRule=UseColor_Specified, specifiedColor=<rgba 0 0 0 0>}
    cornerRadii=<0 0 0 1>
   }
   tintColor {colorUseRule=UseColor_Specified, specifiedColor=<rgba 0.18 0.737 0.878 1>}
   uVRegion {bIsValid=false, max=<0 0>, min=<0 0>}
  }
 }
 CheckBackground_Normal {
  type=SlateBrush
  default {
   bIsDynamicallyLoaded=false
   drawAs=Image
   imageType=NoImage
   mirroring=NoMirror
   resourceName=None
   resourceObject=&T_Icon_ColorPicker_action_bg
   tiling=NoTile
   imageSize=<94 81>
   margin=<ltrb 0 0 0 0>
   outlineSettings {
    bUseBrushTransparency=false
    roundingType=HalfHeightRadius
    width=0
    color {colorUseRule=UseColor_Specified, specifiedColor=<rgba 0 0 0 0>}
    cornerRadii=<0 0 0 1>
   }
   tintColor {colorUseRule=UseColor_Specified, specifiedColor=<rgba 0.094 0.631 0.847 1>}
   uVRegion {bIsValid=false, max=<0 0>, min=<0 0>}
  }
 }
 CheckBackground_Pressed {
  type=SlateBrush
  default {
   bIsDynamicallyLoaded=false
   drawAs=Image
   imageType=NoImage
   mirroring=NoMirror
   resourceName=None
   resourceObject=&T_Icon_ColorPicker_action_bg
   tiling=NoTile
   imageSize=<94 81>
   margin=<ltrb 0 0 0 0>
   outlineSettings {
    bUseBrushTransparency=false
    roundingType=HalfHeightRadius
    width=0
    color {colorUseRule=UseColor_Specified, specifiedColor=<rgba 0 0 0 0>}
    cornerRadii=<0 0 0 1>
   }
   tintColor {colorUseRule=UseColor_Specified, specifiedColor=<rgba 0.302 0.937 1 1>}
   uVRegion {bIsValid=false, max=<0 0>, min=<0 0>}
  }
 }
 CurrentIconSwatchButton {type=CreativeCharacterItemPickerSwatchButton_C, default=null}
 Height {type=number}
 MaxSize {type=Vector2D, default=<75 75>}
 MostRecentlySelectedIconItem {type=IconOptionObjectWrapper, default=null}
 OnIconPickerClosed {}
 OnSwatchSelectionCommitted {}
 ScalableEntrySizeOther {
  type=ScalableFloat
  default {
   registryType=None
   value=1
   curve {curveTable=/Game/Athena/Playlists/Creative/AthenaCompositeGD_Creative, rowName=Creative.UI.CharacterItemPicker.DrawSize.Other}
  }
 }
 ScalableEntrySizePC {
  type=ScalableFloat
  default {
   registryType=None
   value=1
   curve {curveTable=&AthenaCompositeGD_Creative, rowName=Creative.UI.CharacterItemPicker.DrawSize.PC}
  }
 }
 SwatchColCount {type=integer, default=5}
 SwatchCount {type=integer}
]
widgets {
 info {
  inheritedWidgetCount=2
  parentClass=/Script/FortniteUI.FortCreativePropertyEditorColorPickerDialog
  rootWidgetClass=&Overlay
  widgetCount=30
 }
 widgets [
  Overlay_18 {widgetClassPath=&Overlay, widget=Overlay_18}
  UserActionMenuInputButton_Back {
   widgetClassPath=/Game/UI/SocialPanel/UserActionMenuInputButton
   bIsVariable=true
   parent=Overlay_18
   widget=UserActionMenuInputButton_Back
   properties {
    bHideInputActionWithKeyboard=true
    bStopAction=true
    inputPriority=10
    onButtonBaseClicked=(&self:WidgetTree.UserActionMenuInputButton_Back.HandleOnButtonClicked)
    onButtonBaseDoubleClicked=(&self:WidgetTree.UserActionMenuInputButton_Back.HandleOnButtonDoubleClicked)
    onButtonBaseHovered=(&self:WidgetTree.UserActionMenuInputButton_Back.HandleOnButtonHovered)
    onButtonBaseUnhovered=(&self:WidgetTree.UserActionMenuInputButton_Back.HandleOnButtonUnhovered)
    onSelectedChangedBase=(&self:WidgetTree.UserActionMenuInputButton_Back.HandleOnSelectedChanged)
    priority=10
    visibility=Collapsed
    triggeringInputAction {dataTable=&InputActionDataTable, rowName=GenericBack}
   }
   slot {class=OverlaySlot, properties {padding=<ltrb 10 10 10 10>}}
  }
  ScaleBox_48 {widgetClassPath=&ScaleBox, parent=Overlay_18, widget=ScaleBox_48, properties {stretch=Fill}, slot {class=OverlaySlot}}
  SizeBox_WidgetSize {
   widgetClassPath=&SizeBox
   bIsVariable=true
   parent=ScaleBox_48
   widget=SizeBox_WidgetSize
   properties {bOverride_HeightOverride=true, bOverride_WidthOverride=true, heightOverride=418, widthOverride=438}
   slot {class=ScaleBoxSlot, properties {horizontalAlignment=HAlign_Left, verticalAlignment=VAlign_Top}}
  }
  Border_56 {
   widgetClassPath=&Border
   parent=SizeBox_WidgetSize
   widget=Border_56
   properties {
    horizontalAlignment=HAlign_Left
    verticalAlignment=VAlign_Top
    background {
     bIsDynamicallyLoaded=false
     drawAs=Image
     imageType=NoImage
     mirroring=NoMirror
     resourceName=None
     resourceObject=null
     tiling=NoTile
     imageSize=<32 32>
     margin=<ltrb 0 0 0 0>
     outlineSettings {
      bUseBrushTransparency=false
      roundingType=HalfHeightRadius
      width=0
      color {colorUseRule=UseColor_Specified, specifiedColor=<rgba 0 0 0 0>}
      cornerRadii=<0 0 0 0>
     }
     tintColor {colorUseRule=UseColor_Specified, specifiedColor=<rgba 0.0027 0.006 0.0452 1>}
     uVRegion {bIsValid=false, max=<0 0>, min=<0 0>}
    }
    desiredSizeScale=<0 0>
    padding=<ltrb 0 0 0 0>
   }
   slot {class=SizeBoxSlot}
  }
  Image_70 {
   widgetClassPath=&Image
   bIsVariable=true
   parent=Overlay_18
   widget=Image_70
   properties {
    visibility=SelfHitTestInvisible
    brush {
     bIsDynamicallyLoaded=false
     drawAs=Image
     imageType=NoImage
     mirroring=NoMirror
     resourceName=None
     resourceObject=/Game/Athena/UI/InGame/Creative/PropertyEditor/Color_Picker/T_Icon_ColorPicker_current_bg
     tiling=NoTile
     imageSize=<438 80>
     margin=<ltrb 0 0 0 0>
     outlineSettings {
      bUseBrushTransparency=false
      roundingType=HalfHeightRadius
      width=0
      color {colorUseRule=UseColor_Specified, specifiedColor=<rgba 0 0 0 0>}
      cornerRadii=<0 0 0 0>
     }
     tintColor {colorUseRule=UseColor_Specified, specifiedColor=<rgba 1 1 1 1>}
     uVRegion {bIsValid=false, max=<0 0>, min=<0 0>}
    }
   }
   slot {class=OverlaySlot}
  }
  SizeBox_37 {
   widgetClassPath=&SizeBox
   parent=Overlay_18
   widget=SizeBox_37
   properties {bOverride_HeightOverride=true, bOverride_WidthOverride=true, heightOverride=80, widthOverride=438}
   slot {class=OverlaySlot}
  }
  Overlay_0 {widgetClassPath=&Overlay, parent=SizeBox_37, widget=Overlay_0, slot {class=SizeBoxSlot}}
  Image_LeftHighlight {
   widgetClassPath=&Image
   parent=Overlay_0
   widget=Image_LeftHighlight
   properties {brush {
     bIsDynamicallyLoaded=false
     drawAs=Image
     imageType=NoImage
     mirroring=NoMirror
     resourceName=None
     resourceObject=/Game/Athena/UI/InGame/Creative/PropertyEditor/Color_Picker/T_Icon_ColorPicker_highlight
     tiling=NoTile
     imageSize=<10 80>
     margin=<ltrb 0 0 0 0>
     outlineSettings {
      bUseBrushTransparency=false
      roundingType=HalfHeightRadius
      width=0
      color {colorUseRule=UseColor_Specified, specifiedColor=<rgba 0 0 0 0>}
      cornerRadii=<0 0 0 0>
     }
     tintColor {colorUseRule=UseColor_Specified, specifiedColor=<rgba 1 1 1 1>}
     uVRegion {bIsValid=false, max=<0 0>, min=<0 0>}
    }}
   slot {class=OverlaySlot, properties {verticalAlignment=VAlign_Fill}}
  }
  SizeBox_SearchText {
   widgetClassPath=&SizeBox
   parent=Overlay_0
   widget=SizeBox_SearchText
   properties {bOverride_WidthOverride=true, widthOverride=300}
   slot {class=OverlaySlot, properties {verticalAlignment=VAlign_Fill}}
  }
  Overlay_51 {widgetClassPath=&Overlay, parent=SizeBox_SearchText, widget=Overlay_51, slot {class=SizeBoxSlot}}
  Text_SwatchName {
   widgetClassPath=&CommonTextBlock
   bIsVariable=true
   parent=Overlay_51
   widget=Text_SwatchName
   properties {
    scrollStyle=&TextScrollStyle-Base_C
    text=ICON
    font {
     bForceMonospaced=false
     bMaterialIsStencil=false
     fontMaterial=null
     fontObject=&BurbankSmall
     letterSpacing=0
     monospacedWidth=1
     size=20
     skewAmount=0
     typefaceFontName=Bold
     outlineSettings {
      bApplyOutlineToDropShadows=false
      bMiteredCorners=false
      bSeparateFillAlpha=false
      outlineMaterial=null
      outlineSize=0
      outlineColor=<rgba 0 0.0182 0.2031 0.2>
     }
    }
    shadowColorAndOpacity=<rgba 0.0052 0.0052 0.0056 0>
    shadowOffset=<2 1>
   }
   slot {class=OverlaySlot, properties {verticalAlignment=VAlign_Center, padding=<ltrb 20 0 16 0>}}
  }
  EditableText_Search {
   widgetClassPath=&EditableTextBox
   bInherited=true
   parent=Overlay_51
   widget=EditableText_Search
   properties {
    hintText=Type To Search Icon
    virtualKeyboardTrigger=OnAllFocusEvents
    visibility=Collapsed
    widgetStyle {
     backgroundColor {colorUseRule=UseColor_Specified, specifiedColor=<rgba 1 1 1 0>}
     backgroundImageFocused {
      bIsDynamicallyLoaded=false
      drawAs=Box
      imageType=FullColor
      mirroring=NoMirror
      resourceName=../../../Engine/Content/Slate/Common/TextBox_Hovered.png
      resourceObject=null
      tiling=NoTile
      imageSize=<0 0>
      margin=<ltrb 0.25 0.25 0.25 0.25>
      outlineSettings {
       bUseBrushTransparency=false
       roundingType=HalfHeightRadius
       width=0
       color {colorUseRule=UseColor_Specified, specifiedColor=<rgba 0 0 0 0>}
       cornerRadii=<0 0 0 0>
      }
      tintColor {colorUseRule=UseColor_Specified, specifiedColor=<rgba 1 1 1 1>}
      uVRegion {bIsValid=false, max=<0 0>, min=<0 0>}
     }
     backgroundImageHovered {
      bIsDynamicallyLoaded=false
      drawAs=Box
      imageType=FullColor
      mirroring=NoMirror
      resourceName=../../../Engine/Content/Slate/Common/TextBox_Hovered.png
      resourceObject=null
      tiling=NoTile
      imageSize=<0 0>
      margin=<ltrb 0.25 0.25 0.25 0.25>
      outlineSettings {
       bUseBrushTransparency=false
       roundingType=HalfHeightRadius
       width=0
       color {colorUseRule=UseColor_Specified, specifiedColor=<rgba 0 0 0 0>}
       cornerRadii=<0 0 0 0>
      }
      tintColor {colorUseRule=UseColor_Specified, specifiedColor=<rgba 1 1 1 1>}
      uVRegion {bIsValid=false, max=<0 0>, min=<0 0>}
     }
     backgroundImageNormal {
      bIsDynamicallyLoaded=false
      drawAs=Box
      imageType=FullColor
      mirroring=NoMirror
      resourceName=../../../Engine/Content/Slate/Common/TextBox.png
      resourceObject=null
      tiling=NoTile
      imageSize=<0 0>
      margin=<ltrb 0.25 0.25 0.25 0.25>
      outlineSettings {
       bUseBrushTransparency=false
       roundingType=HalfHeightRadius
       width=0
       color {colorUseRule=UseColor_Specified, specifiedColor=<rgba 0 0 0 0>}
       cornerRadii=<0 0 0 0>
      }
      tintColor {colorUseRule=UseColor_Specified, specifiedColor=<rgba 1 1 1 1>}
      uVRegion {bIsValid=false, max=<0 0>, min=<0 0>}
     }
     backgroundImageReadOnly {
      bIsDynamicallyLoaded=false
      drawAs=Box
      imageType=FullColor
      mirroring=NoMirror
      resourceName=../../../Engine/Content/Slate/Common/TextBox_ReadOnly.png
      resourceObject=null
      tiling=NoTile
      imageSize=<0 0>
      margin=<ltrb 0.25 0.25 0.25 0.25>
      outlineSettings {
       bUseBrushTransparency=false
       roundingType=HalfHeightRadius
       width=0
       color {colorUseRule=UseColor_Specified, specifiedColor=<rgba 0 0 0 0>}
       cornerRadii=<0 0 0 0>
      }
      tintColor {colorUseRule=UseColor_Specified, specifiedColor=<rgba 1 1 1 1>}
      uVRegion {bIsValid=false, max=<0 0>, min=<0 0>}
     }
     focusedForegroundColor {colorUseRule=UseColor_Specified, specifiedColor=<rgba 1 1 1 1>}
     foregroundColor {colorUseRule=UseColor_Specified, specifiedColor=<rgba 1 1 1 1>}
     hScrollBarPadding=<ltrb 0 0 0 0>
     padding=<ltrb 36 0 36 0>
     readOnlyForegroundColor {colorUseRule=UseColor_Foreground, specifiedColor=<rgba 1 0 1 1>}
     scrollBarStyle {
      thickness=16
      draggedThumbImage {
       bIsDynamicallyLoaded=false
       drawAs=Box
       imageType=FullColor
       mirroring=NoMirror
       resourceName=../../../Engine/Content/Slate/Common/Scrollbar_Thumb.png
       resourceObject=null
       tiling=NoTile
       imageSize=<0 0>
       margin=<ltrb 0.25 0.25 0.25 0.25>
       outlineSettings {
        bUseBrushTransparency=false
        roundingType=HalfHeightRadius
        width=0
        color {colorUseRule=UseColor_Specified, specifiedColor=<rgba 0 0 0 0>}
        cornerRadii=<0 0 0 0>
       }
       tintColor {colorUseRule=UseColor_Specified, specifiedColor=<rgba 1 1 1 1>}
       uVRegion {bIsValid=false, max=<0 0>, min=<0 0>}
      }
      horizontalBackgroundImage {
       bIsDynamicallyLoaded=false
       drawAs=NoDrawType
       imageType=NoImage
       mirroring=NoMirror
       resourceName=None
       resourceObject=null
       tiling=NoTile
       imageSize=<0 0>
       margin=<ltrb 0 0 0 0>
       outlineSettings {
        bUseBrushTransparency=false
        roundingType=HalfHeightRadius
        width=0
        color {colorUseRule=UseColor_Specified, specifiedColor=<rgba 0 0 0 0>}
        cornerRadii=<0 0 0 0>
       }
       tintColor {colorUseRule=UseColor_Specified, specifiedColor=<rgba 1 1 1 1>}
       uVRegion {bIsValid=false, max=<0 0>, min=<0 0>}
      }
      horizontalBottomSlotImage {
       bIsDynamicallyLoaded=false
       drawAs=Image
       imageType=FullColor
       mirroring=NoMirror
       resourceName=../../../Engine/Content/Slate/Common/Scrollbar_Background_Horizontal.png
       resourceObject=null
       tiling=NoTile
       imageSize=<8 8>
       margin=<ltrb 0 0 0 0>
       outlineSettings {
        bUseBrushTransparency=false
        roundingType=HalfHeightRadius
        width=0
        color {colorUseRule=UseColor_Specified, specifiedColor=<rgba 0 0 0 0>}
        cornerRadii=<0 0 0 0>
       }
       tintColor {colorUseRule=UseColor_Specified, specifiedColor=<rgba 1 1 1 1>}
       uVRegion {bIsValid=false, max=<0 0>, min=<0 0>}
      }
      horizontalTopSlotImage {
       bIsDynamicallyLoaded=false
       drawAs=Image
       imageType=FullColor
       mirroring=NoMirror
       resourceName=../../../Engine/Content/Slate/Common/Scrollbar_Background_Horizontal.png
       resourceObject=null
       tiling=NoTile
       imageSize=<8 8>
       margin=<ltrb 0 0 0 0>
       outlineSettings {
        bUseBrushTransparency=false
        roundingType=HalfHeightRadius
        width=0
        color {colorUseRule=UseColor_Specified, specifiedColor=<rgba 0 0 0 0>}
        cornerRadii=<0 0 0 0>
       }
       tintColor {colorUseRule=UseColor_Specified, specifiedColor=<rgba 1 1 1 1>}
       uVRegion {bIsValid=false, max=<0 0>, min=<0 0>}
      }
      hoveredThumbImage {
       bIsDynamicallyLoaded=false
       drawAs=Box
       imageType=FullColor
       mirroring=NoMirror
       resourceName=../../../Engine/Content/Slate/Common/Scrollbar_Thumb.png
       resourceObject=null
       tiling=NoTile
       imageSize=<0 0>
       margin=<ltrb 0.25 0.25 0.25 0.25>
       outlineSettings {
        bUseBrushTransparency=false
        roundingType=HalfHeightRadius
        width=0
        color {colorUseRule=UseColor_Specified, specifiedColor=<rgba 0 0 0 0>}
        cornerRadii=<0 0 0 0>
       }
       tintColor {colorUseRule=UseColor_Specified, specifiedColor=<rgba 1 1 1 1>}
       uVRegion {bIsValid=false, max=<0 0>, min=<0 0>}
      }
      normalThumbImage {
       bIsDynamicallyLoaded=false
       drawAs=Box
       imageType=FullColor
       mirroring=NoMirror
       resourceName=../../../Engine/Content/Slate/Common/Scrollbar_Thumb.png
       resourceObject=null
       tiling=NoTile
       imageSize=<0 0>
       margin=<ltrb 0.25 0.25 0.25 0.25>
       outlineSettings {
        bUseBrushTransparency=false
        roundingType=HalfHeightRadius
        width=0
        color {colorUseRule=UseColor_Specified, specifiedColor=<rgba 0 0 0 0>}
        cornerRadii=<0 0 0 0>
       }
       tintColor {colorUseRule=UseColor_Specified, specifiedColor=<rgba 1 1 1 1>}
       uVRegion {bIsValid=false, max=<0 0>, min=<0 0>}
      }
      verticalBackgroundImage {
       bIsDynamicallyLoaded=false
       drawAs=NoDrawType
       imageType=NoImage
       mirroring=NoMirror
       resourceName=None
       resourceObject=null
       tiling=NoTile
       imageSize=<0 0>
       margin=<ltrb 0 0 0 0>
       outlineSettings {
        bUseBrushTransparency=false
        roundingType=HalfHeightRadius
        width=0
        color {colorUseRule=UseColor_Specified, specifiedColor=<rgba 0 0 0 0>}
        cornerRadii=<0 0 0 0>
       }
       tintColor {colorUseRule=UseColor_Specified, specifiedColor=<rgba 1 1 1 1>}
       uVRegion {bIsValid=false, max=<0 0>, min=<0 0>}
      }
      verticalBottomSlotImage {
       bIsDynamicallyLoaded=false
       drawAs=Image
       imageType=FullColor
       mirroring=NoMirror
       resourceName=../../../Engine/Content/Slate/Common/Scrollbar_Background_Vertical.png
       resourceObject=null
       tiling=NoTile
       imageSize=<8 8>
       margin=<ltrb 0 0 0 0>
       outlineSettings {
        bUseBrushTransparency=false
        roundingType=HalfHeightRadius
        width=0
        color {colorUseRule=UseColor_Specified, specifiedColor=<rgba 0 0 0 0>}
        cornerRadii=<0 0 0 0>
       }
       tintColor {colorUseRule=UseColor_Specified, specifiedColor=<rgba 1 1 1 1>}
       uVRegion {bIsValid=false, max=<0 0>, min=<0 0>}
      }
      verticalTopSlotImage {
       bIsDynamicallyLoaded=false
       drawAs=Image
       imageType=FullColor
       mirroring=NoMirror
       resourceName=../../../Engine/Content/Slate/Common/Scrollbar_Background_Vertical.png
       resourceObject=null
       tiling=NoTile
       imageSize=<8 8>
       margin=<ltrb 0 0 0 0>
       outlineSettings {
        bUseBrushTransparency=false
        roundingType=HalfHeightRadius
        width=0
        color {colorUseRule=UseColor_Specified, specifiedColor=<rgba 0 0 0 0>}
        cornerRadii=<0 0 0 0>
       }
       tintColor {colorUseRule=UseColor_Specified, specifiedColor=<rgba 1 1 1 1>}
       uVRegion {bIsValid=false, max=<0 0>, min=<0 0>}
      }
     }
     textStyle {
      overflowPolicy=Clip
      transformPolicy=None
      colorAndOpacity {colorUseRule=UseColor_Foreground, specifiedColor=<rgba 1 0 1 1>}
      font {
       bForceMonospaced=false
       bMaterialIsStencil=false
       fontMaterial=null
       fontObject=&BurbankSmall
       letterSpacing=0
       monospacedWidth=1
       size=14
       skewAmount=0
       typefaceFontName=Bold
       outlineSettings {
        bApplyOutlineToDropShadows=false
        bMiteredCorners=false
        bSeparateFillAlpha=false
        outlineMaterial=null
        outlineSize=0
        outlineColor=<rgba 0 0 0 1>
       }
      }
      highlightColor {colorUseRule=UseColor_Specified, specifiedColor=<rgba 0.02 0.3 0 1>}
      highlightShape {
       bIsDynamicallyLoaded=false
       drawAs=Box
       imageType=FullColor
       mirroring=NoMirror
       resourceName=../../../Engine/Content/Slate/Common/TextBlockHighlightShape.png
       resourceObject=null
       tiling=NoTile
       imageSize=<0 0>
       margin=<ltrb 0.375 0.375 0.375 0.375>
       outlineSettings {
        bUseBrushTransparency=false
        roundingType=HalfHeightRadius
        width=0
        color {colorUseRule=UseColor_Specified, specifiedColor=<rgba 0 0 0 0>}
        cornerRadii=<0 0 0 0>
       }
       tintColor {colorUseRule=UseColor_Specified, specifiedColor=<rgba 1 1 1 1>}
       uVRegion {bIsValid=false, max=<0 0>, min=<0 0>}
      }
      selectedBackgroundColor {colorUseRule=UseColor_Foreground, specifiedColor=<rgba 1 0 1 1>}
      shadowColorAndOpacity=<rgba 0 0 0 1>
      shadowOffset=<0 0>
      strikeBrush {
       bIsDynamicallyLoaded=false
       drawAs=Image
       imageType=NoImage
       mirroring=NoMirror
       resourceName=None
       resourceObject=null
       tiling=NoTile
       imageSize=<32 32>
       margin=<ltrb 0 0 0 0>
       outlineSettings {
        bUseBrushTransparency=false
        roundingType=HalfHeightRadius
        width=0
        color {colorUseRule=UseColor_Specified, specifiedColor=<rgba 0 0 0 0>}
        cornerRadii=<0 0 0 0>
       }
       tintColor {colorUseRule=UseColor_Specified, specifiedColor=<rgba 1 1 1 1>}
       uVRegion {bIsValid=false, max=<0 0>, min=<0 0>}
      }
      underlineBrush {
       bIsDynamicallyLoaded=false
       drawAs=Image
       imageType=NoImage
       mirroring=NoMirror
       resourceName=None
       resourceObject=null
       tiling=NoTile
       imageSize=<32 32>
       margin=<ltrb 0 0 0 0>
       outlineSettings {
        bUseBrushTransparency=false
        roundingType=HalfHeightRadius
        width=0
        color {colorUseRule=UseColor_Specified, specifiedColor=<rgba 0 0 0 0>}
        cornerRadii=<0 0 0 0>
       }
       tintColor {colorUseRule=UseColor_Specified, specifiedColor=<rgba 1 1 1 1>}
       uVRegion {bIsValid=false, max=<0 0>, min=<0 0>}
      }
     }
     vScrollBarPadding=<ltrb 0 0 0 0>
    }
   }
   slot {class=OverlaySlot, properties {horizontalAlignment=HAlign_Fill, verticalAlignment=VAlign_Fill}}
  }
  SearchImage {
   widgetClassPath=&Image
   bIsVariable=true
   parent=Overlay_0
   widget=SearchImage
   properties {
    visibility=SelfHitTestInvisible
    brush {
     bIsDynamicallyLoaded=false
     drawAs=Image
     imageType=NoImage
     mirroring=NoMirror
     resourceName=None
     resourceObject=/Game/Athena/UI/InGame/Creative/PropertyEditor/Art/Icon-PropertyEditor-Search
     tiling=NoTile
     imageSize=<32 32>
     margin=<ltrb 0 0 0 0>
     outlineSettings {
      bUseBrushTransparency=false
      roundingType=HalfHeightRadius
      width=0
      color {colorUseRule=UseColor_Specified, specifiedColor=<rgba 0 0 0 0>}
      cornerRadii=<0 0 0 0>
     }
     tintColor {colorUseRule=UseColor_Specified, specifiedColor=<rgba 1 1 1 0.6>}
     uVRegion {bIsValid=false, max=<0 0>, min=<0 0>}
    }
    renderTransform {angle=0, scale=<-1 1>, shear=<0 0>, translation=<0 0>}
   }
   slot {class=OverlaySlot, properties {horizontalAlignment=HAlign_Right, verticalAlignment=VAlign_Center, padding=<ltrb 0 0 116 0>}}
  }
  CommonHardwareVisibilityBorder_1 {
   widgetClassPath=&CommonHardwareVisibilityBorder
   parent=Overlay_0
   widget=CommonHardwareVisibilityBorder_1
   properties {
    horizontalAlignment=HAlign_Left
    visibilityQuery {
     autoDescription=" ANY( Input.MouseAndKeyboard, Input.Touch )"
     tokenStreamVersion=0
     userDescription=Mouse and Keyboard
     queryTokenStream [0,1,1,2,0,1]
     tagDictionary [{tagName=Input.MouseAndKeyboard}, {tagName=Input.Touch}]
    }
   }
   slot {class=OverlaySlot, properties {horizontalAlignment=HAlign_Fill, verticalAlignment=VAlign_Fill}}
  }
  SizeBox_261 {
   widgetClassPath=&SizeBox
   parent=CommonHardwareVisibilityBorder_1
   widget=SizeBox_261
   properties {bOverride_WidthOverride=true, widthOverride=340}
   slot {class=BorderSlot, properties {horizontalAlignment=HAlign_Left, padding=<ltrb 0 0 0 0>}}
  }
  Button_Search {
   widgetClassPath=&Button
   parent=SizeBox_261
   widget=Button_Search
   properties {
    isFocusable=false
    widgetStyle {
     clickedSlateSound {resourceObject=null}
     disabled {
      bIsDynamicallyLoaded=false
      drawAs=NoDrawType
      imageType=NoImage
      mirroring=NoMirror
      resourceName=None
      resourceObject=null
      tiling=NoTile
      imageSize=<0 0>
      margin=<ltrb 0 0 0 0>
      outlineSettings {
       bUseBrushTransparency=false
       roundingType=HalfHeightRadius
       width=0
       color {colorUseRule=UseColor_Specified, specifiedColor=<rgba 0 0 0 0>}
       cornerRadii=<0 0 0 0>
      }
      tintColor {colorUseRule=UseColor_Specified, specifiedColor=<rgba 1 1 1 1>}
      uVRegion {bIsValid=false, max=<0 0>, min=<0 0>}
     }
     disabledForeground {colorUseRule=UseColor_Specified, specifiedColor=<rgba 0.5271 0.5271 0.5271 1>}
     hovered {
      bIsDynamicallyLoaded=false
      drawAs=NoDrawType
      imageType=NoImage
      mirroring=NoMirror
      resourceName=None
      resourceObject=null
      tiling=NoTile
      imageSize=<32 32>
      margin=<ltrb 0 0 0 0>
      outlineSettings {
       bUseBrushTransparency=true
       roundingType=FixedRadius
       width=1
       color {colorUseRule=UseColor_Specified, specifiedColor=<rgba 0.7243 0.7243 0.7243 1>}
       cornerRadii=<4 4 4 4>
      }
      tintColor {colorUseRule=UseColor_Specified, specifiedColor=<rgba 0.7243 0.7243 0.7243 1>}
      uVRegion {bIsValid=false, max=<0 0>, min=<0 0>}
     }
     hoveredForeground {colorUseRule=UseColor_Specified, specifiedColor=<rgba 0 1 1 1>}
     hoveredSlateSound {resourceObject=null}
     normal {
      bIsDynamicallyLoaded=false
      drawAs=NoDrawType
      imageType=NoImage
      mirroring=NoMirror
      resourceName=None
      resourceObject=null
      tiling=NoTile
      imageSize=<32 32>
      margin=<ltrb 0 0 0 0>
      outlineSettings {
       bUseBrushTransparency=true
       roundingType=FixedRadius
       width=1
       color {colorUseRule=UseColor_Specified, specifiedColor=<rgba 0.6951 0.6951 0.6951 1>}
       cornerRadii=<4 4 4 4>
      }
      tintColor {colorUseRule=UseColor_Specified, specifiedColor=<rgba 0.4955 0.4955 0.4955 1>}
      uVRegion {bIsValid=false, max=<0 0>, min=<0 0>}
     }
     normalForeground {colorUseRule=UseColor_Specified, specifiedColor=<rgba 0 1 1 1>}
     normalPadding=<ltrb 12 1.5 12 1.5>
     pressed {
      bIsDynamicallyLoaded=false
      drawAs=NoDrawType
      imageType=NoImage
      mirroring=NoMirror
      resourceName=None
      resourceObject=null
      tiling=NoTile
      imageSize=<32 32>
      margin=<ltrb 0 0 0 0>
      outlineSettings {
       bUseBrushTransparency=true
       roundingType=FixedRadius
       width=1
       color {colorUseRule=UseColor_Specified, specifiedColor=<rgba 0.7243 0.7243 0.7243 1>}
       cornerRadii=<4 4 4 4>
      }
      tintColor {colorUseRule=UseColor_Specified, specifiedColor=<rgba 0.3843 0.3843 0.3843 1>}
      uVRegion {bIsValid=false, max=<0 0>, min=<0 0>}
     }
     pressedForeground {colorUseRule=UseColor_Specified, specifiedColor=<rgba 0 1 1 1>}
     pressedPadding=<ltrb 12 1.5 12 1.5>
     pressedSlateSound {resourceObject=null}
    }
   }
   slot {class=SizeBoxSlot}
  }
  Button_CheckmarkBackground {
   widgetClassPath=&Button
   parent=Overlay_0
   widget=Button_CheckmarkBackground
   properties {
    isFocusable=false
    widgetStyle {
     clickedSlateSound {resourceObject=null}
     disabled {
      bIsDynamicallyLoaded=false
      drawAs=NoDrawType
      imageType=NoImage
      mirroring=NoMirror
      resourceName=None
      resourceObject=null
      tiling=NoTile
      imageSize=<0 0>
      margin=<ltrb 0 0 0 0>
      outlineSettings {
       bUseBrushTransparency=false
       roundingType=HalfHeightRadius
       width=0
       color {colorUseRule=UseColor_Specified, specifiedColor=<rgba 0 0 0 0>}
       cornerRadii=<0 0 0 0>
      }
      tintColor {colorUseRule=UseColor_Specified, specifiedColor=<rgba 1 1 1 1>}
      uVRegion {bIsValid=false, max=<0 0>, min=<0 0>}
     }
     disabledForeground {colorUseRule=UseColor_Specified, specifiedColor=<rgba 0.5271 0.5271 0.5271 1>}
     hovered {
      bIsDynamicallyLoaded=false
      drawAs=RoundedBox
      imageType=NoImage
      mirroring=NoMirror
      resourceName=None
      resourceObject=&T_Icon_ColorPicker_action_bg
      tiling=NoTile
      imageSize=<94 81>
      margin=<ltrb 0 0 0 0>
      outlineSettings {
       bUseBrushTransparency=true
       roundingType=FixedRadius
       width=1
       color {colorUseRule=UseColor_Specified, specifiedColor=<rgba 0.7243 0.7243 0.7243 1>}
       cornerRadii=<4 4 4 4>
      }
      tintColor {colorUseRule=UseColor_Specified, specifiedColor=<rgba 1 1 1 0.5>}
      uVRegion {bIsValid=false, max=<0 0>, min=<0 0>}
     }
     hoveredForeground {colorUseRule=UseColor_Specified, specifiedColor=<rgba 0 1 1 1>}
     hoveredSlateSound {resourceObject=null}
     normal {
      bIsDynamicallyLoaded=false
      drawAs=RoundedBox
      imageType=NoImage
      mirroring=NoMirror
      resourceName=None
      resourceObject=&T_Icon_ColorPicker_action_bg
      tiling=NoTile
      imageSize=<94 81>
      margin=<ltrb 0 0 0 0>
      outlineSettings {
       bUseBrushTransparency=true
       roundingType=FixedRadius
       width=1
       color {colorUseRule=UseColor_Specified, specifiedColor=<rgba 0.6951 0.6951 0.6951 1>}
       cornerRadii=<4 4 4 4>
      }
      tintColor {colorUseRule=UseColor_Specified, specifiedColor=<rgba 1 1 1 0.25>}
      uVRegion {bIsValid=false, max=<0 0>, min=<0 0>}
     }
     normalForeground {colorUseRule=UseColor_Specified, specifiedColor=<rgba 0 1 1 1>}
     normalPadding=<ltrb 0 0 0 0>
     pressed {
      bIsDynamicallyLoaded=false
      drawAs=RoundedBox
      imageType=NoImage
      mirroring=NoMirror
      resourceName=None
      resourceObject=&T_Icon_ColorPicker_action_bg
      tiling=NoTile
      imageSize=<94 81>
      margin=<ltrb 0 0 0 0>
      outlineSettings {
       bUseBrushTransparency=true
       roundingType=FixedRadius
       width=1
       color {colorUseRule=UseColor_Specified, specifiedColor=<rgba 0.7243 0.7243 0.7243 1>}
       cornerRadii=<4 4 4 4>
      }
      tintColor {colorUseRule=UseColor_Specified, specifiedColor=<rgba 1 1 1 0.5>}
      uVRegion {bIsValid=false, max=<0 0>, min=<0 0>}
     }
     pressedForeground {colorUseRule=UseColor_Specified, specifiedColor=<rgba 0 1 1 1>}
     pressedPadding=<ltrb 0 0 0 0>
     pressedSlateSound {resourceObject=null}
    }
   }
   slot {class=OverlaySlot, properties {horizontalAlignment=HAlign_Right, verticalAlignment=VAlign_Fill}}
  }
  SizeBox_1 {
   widgetClassPath=&SizeBox
   parent=Overlay_0
   widget=SizeBox_1
   properties {bOverride_WidthOverride=true, visibility=HitTestInvisible, widthOverride=70}
   slot {class=OverlaySlot, properties {horizontalAlignment=HAlign_Right, verticalAlignment=VAlign_Center}}
  }
  HorizontalBox_312 {widgetClassPath=&HorizontalBox, parent=SizeBox_1, widget=HorizontalBox_312, slot {class=SizeBoxSlot}}
  UserActionMenuInputButton_Accept {
   widgetClassPath=&UserActionMenuInputButton
   bIsVariable=true
   parent=HorizontalBox_312
   widget=UserActionMenuInputButton_Accept
   properties {
    bHideInputActionWithKeyboard=true
    onButtonBaseClicked=(&self:WidgetTree.UserActionMenuInputButton_Accept.HandleOnButtonClicked)
    onButtonBaseDoubleClicked=(&self:WidgetTree.UserActionMenuInputButton_Accept.HandleOnButtonDoubleClicked)
    onButtonBaseHovered=(&self:WidgetTree.UserActionMenuInputButton_Accept.HandleOnButtonHovered)
    onButtonBaseUnhovered=(&self:WidgetTree.UserActionMenuInputButton_Accept.HandleOnButtonUnhovered)
    onSelectedChangedBase=(&self:WidgetTree.UserActionMenuInputButton_Accept.HandleOnSelectedChanged)
    triggeringInputAction {dataTable=&InputActionDataTable, rowName=VoteYes}
   }
   slot {class=HorizontalBoxSlot}
  }
  Overlay_1 {
   widgetClassPath=&Overlay
   parent=HorizontalBox_312
   widget=Overlay_1
   slot {class=HorizontalBoxSlot, properties {size {sizeRule=Fill, value=1}}}
  }
  Image_Checkmark {
   widgetClassPath=&Image
   parent=Overlay_1
   widget=Image_Checkmark
   properties {brush {
     bIsDynamicallyLoaded=false
     drawAs=Image
     imageType=NoImage
     mirroring=NoMirror
     resourceName=None
     resourceObject=/Game/UI/InGame/Mission/MissionActivation/Textures/T-Icon-VoteYes
     tiling=NoTile
     imageSize=<32 32>
     margin=<ltrb 0 0 0 0>
     outlineSettings {
      bUseBrushTransparency=false
      roundingType=HalfHeightRadius
      width=0
      color {colorUseRule=UseColor_Specified, specifiedColor=<rgba 0 0 0 0>}
      cornerRadii=<0 0 0 0>
     }
     tintColor {colorUseRule=UseColor_Specified, specifiedColor=<rgba 1 1 1 1>}
     uVRegion {bIsValid=false, max=<0 0>, min=<0 0>}
    }}
   slot {class=OverlaySlot, properties {padding=<ltrb 5 0 0 0>}}
  }
  IconSwatchArea {
   widgetClassPath=&SizeBox
   parent=Overlay_18
   widget=IconSwatchArea
   properties {bOverride_HeightOverride=true, bOverride_WidthOverride=true, heightOverride=250, widthOverride=400}
   slot {class=OverlaySlot, properties {padding=<ltrb 20 90 0 0>}}
  }
  TileView_Swatches {
   widgetClassPath=/Script/CommonUI.CommonTileView
   bInherited=true
   bIsVariable=true
   parent=IconSwatchArea
   widget=TileView_Swatches
   properties {
    bWrapHorizontalNavigation=true
    entryHeight=76
    entryWidgetClass=&/CreativeCharacterItemPickerSwatchButton.CreativeCharacterItemPickerSwatchButton_C
    entryWidth=76
    navigation=&self:WidgetTree.TileView_Swatches.WidgetNavigation_0
    numDesignerPreviewEntries=12
   }
   slot {class=SizeBoxSlot}
  }
  SizeBox_SearchInput {
   widgetClassPath=&SizeBox
   bIsVariable=true
   parent=Overlay_18
   widget=SizeBox_SearchInput
   properties {bOverride_HeightOverride=true, heightOverride=68}
   slot {class=OverlaySlot, properties {horizontalAlignment=HAlign_Fill, padding=<ltrb 0 350 0 0>}}
  }
  CommonHardwareVisibilityBorder_0 {
   widgetClassPath=&CommonHardwareVisibilityBorder
   parent=SizeBox_SearchInput
   widget=CommonHardwareVisibilityBorder_0
   properties {
    hiddenType=Hidden
    padding=<ltrb 32 0 32 0>
    visibilityQuery {
     autoDescription=" ANY( Input.Gamepad )"
     tokenStreamVersion=0
     userDescription=Gamepad
     queryTokenStream [0,1,1,1,0]
     tagDictionary [{tagName=Input.Gamepad}]
    }
   }
   slot {class=SizeBoxSlot}
  }
  HorizontalBox_75 {
   widgetClassPath=&HorizontalBox
   parent=CommonHardwareVisibilityBorder_0
   widget=HorizontalBox_75
   slot {class=BorderSlot, properties {padding=<ltrb 32 0 32 0>}}
  }
  SearchActionWidget {
   widgetClassPath=&CommonActionWidget
   bIsVariable=true
   parent=HorizontalBox_75
   widget=SearchActionWidget
   properties {visibility=SelfHitTestInvisible}
   slot {class=HorizontalBoxSlot, properties {verticalAlignment=VAlign_Center}}
  }
  CommonTextBlock_60 {
   widgetClassPath=&CommonTextBlock
   parent=HorizontalBox_75
   widget=CommonTextBlock_60
   properties {
    scrollStyle=&TextScrollStyle-Base_C
    text=Search
    font {
     bForceMonospaced=false
     bMaterialIsStencil=false
     fontMaterial=null
     fontObject=&BurbankSmall
     letterSpacing=0
     monospacedWidth=1
     size=14
     skewAmount=0
     typefaceFontName=Bold
     outlineSettings {
      bApplyOutlineToDropShadows=false
      bMiteredCorners=false
      bSeparateFillAlpha=false
      outlineMaterial=null
      outlineSize=0
      outlineColor=<rgba 0 0 0 1>
     }
    }
    shadowOffset=<0 0>
   }
   slot {class=HorizontalBoxSlot, properties {verticalAlignment=VAlign_Center, padding=<ltrb 24 0 0 0>}}
  }
 ]
}
